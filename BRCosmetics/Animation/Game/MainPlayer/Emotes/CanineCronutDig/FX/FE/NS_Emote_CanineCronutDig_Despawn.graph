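# Niagara despawn effect for the CanineCronutDig emote, bursting dust and debris from the emote's rock skeleton bones.
# Dust sprites spawn at main rock bones in four burst groups that vary size, lifetime and alpha, with time-of-day tinting and curve-driven material parameters.
# Sprite and tumbling mesh debris sample main and outer rock bones, launch upward, then shrink under increasing gravity.
# A front-end preview flag switches sprite debris color.
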NiagaraSystem /BRCosmetics/Animation/Game/MainPlayer/Emotes/CanineCronutDig/FX/FE/NS_Emote_CanineCronutDig_Despawn
dependencies {
 usedDataInterfaces [&NiagaraDataInterfaceCurve, &NiagaraDataInterfaceColorCurve]
 usedDynamicInputs []
 usedModules [&SystemState]
}
emitters {
 Dust001 {
  emitterUpdateScript [
   &EmitterState()
   &SpawnBurst_Instantaneous(
    Spawn Count=&RandomRangeInt(Minimum=8, Maximum=15)
    Use Spawn Probability=false
    UseLoopCountLimit=false
   )
   &SpawnBurst_Instantaneous(
    Spawn Count=&RandomRangeInt(Minimum=8, Maximum=15)
    Spawn Group=2
    Use Spawn Probability=false
    UseLoopCountLimit=false
   ) {module=SpawnBurst_Instantaneous002}
   &SpawnBurst_Instantaneous(
    Spawn Count=&RandomRangeInt(Minimum=8, Maximum=15)
    Spawn Group=1
    Use Spawn Probability=false
    UseLoopCountLimit=false
   ) {module=SpawnBurst_Instantaneous001}
   &SpawnBurst_Instantaneous(
    Spawn Count=&RandomRangeInt(Minimum=8, Maximum=15)
    Spawn Group=3
    Use Spawn Probability=false
    UseLoopCountLimit=false
   ) {module=SpawnBurst_Instantaneous003}
  ]
  particleSpawnScript [
   /SetVariables(Particles.VisibilityTag=&MakeCustomIntFromBool(
     Bool=&SetBoolByIntComparison(A=&SpawnGroup(), B=2, Comparison Type=::NewEnumerator2)
     True INT=1
     False INT=0
    )) {isSetParameters=true}
   /SetVariables(
    Particles.SizeMax:NiagaraFloat=&MakeCustomFloatFromBool(
     Bool=&SetBoolByIntComparison(A=&SpawnGroup(), B=2, Comparison Type=::NewEnumerator2)
     True Float=&RandomRangeFloat(Minimum=45, Maximum=65)
     False Float=&RandomRangeFloat(Minimum=155, Maximum=185)
    )
    Particles.SizeMin:NiagaraFloat=&MakeCustomFloatFromBool(
     Bool=&SetBoolByIntComparison(A=&SpawnGroup(), B=2, Comparison Type=::NewEnumerator2)
     True Float=&RandomRangeFloat(Minimum=65, Maximum=85)
     False Float=&RandomRangeFloat(Minimum=165, Maximum=185)
    )
   ) {isSetParameters=true}
   /SetVariables(Particles.SizeCurve:NiagaraFloat=&Lerp_Float(
     A=$Particles.SizeMin
     B=$Particles.SizeMax
     Alpha=&FloatFromCurve(
      FloatCurve=DataInterface() {propertyValues {Curve {keys [
          {interpMode=RCIM_Cubic, tangentMode=RCTM_Break, time=0, value=0}
          {arriveTangentWeight=0.9217, interpMode=RCIM_Cubic, tangentMode=RCTM_Break, time=1, value=1}
         ]}}}
      CurveIndex=&RandomRangeFloat(Minimum=0, Maximum=1)
     )
    )) {isSetParameters=true}
   &InitializeParticle(
    Color=&MakeLinearColorFromVectorAndFloat(
     "Vector (RGB)"=&MakeVectorFromLinearColorRGB(Color=&FNTimeOfDayColor(
       Night Color=<rgba 0.1406 0.0337 0.0293 1>
       Morning Color=<rgba 0.1198 0.0323 0.0287 1>
       Day Color=<rgba 0.1198 0.0244 0.0205 1>
       Evening Color=<rgba 0.1198 0.0244 0.0205 1>
      ))
     "Float (Alpha)"=&MaskFloatBySpawnGroup(
      Masked Float=&RandomRangeFloat(Minimum=0.25, Maximum=0.25)
      Passthrough Float=&MaskFloatBySpawnGroup(
       Masked Float=&RandomRangeFloat(Minimum=1, Maximum=1)
       Passthrough Float=&RandomRangeFloat(Minimum=0.3, Maximum=0.3)
       Spawn Group Mask=1
      )
     )
    )
    Lifetime=&MaskFloatBySpawnGroup(
     Masked Float=&RandomRangeFloat(Minimum=1, Maximum=2)
     Passthrough Float=&MaskFloatBySpawnGroup(Masked Float=&RandomRangeFloat(Minimum=0.5, Maximum=1), Passthrough Float=1, Spawn Group Mask=1)
    )
    Sprite Rotation Mode=::NewEnumerator2
    Sprite Size Mode=::NewEnumerator3
    Uniform Sprite Size=&Multiply_Float(A=0.5, B=$Particles.SizeCurve)
    UsePositionOffset=false
    Write Material Random=true
   )
   &SkeletalMeshLocation(
    Bone Sampling Mode=::NewEnumerator0
    MaskSpawnGroup=false
    OverrideRandomSeed=false
    Skeletal Mesh=DataInterface() {propertyValues {
      DefaultMesh=null
      ExcludeBoneName=root
      OptionalComponentClass=null
      PreviewMesh=null
      SkinningMode=SkinOnTheFly
      SoftSourceActor=null
      SourceMode=Default
      UvSetIndex=0
      WholeMeshLOD=-1
      bExcludeBone=false
      bReadDeformedGeometry=true
      bRequireCurrentFrameData=true
      ComponentTags []
      FilteredBones [JNT_MainRock_01, JNT_MainRock_02, JNT_MainRock_03]
      FilteredSockets []
      MeshUserParameter {parameter {name=None, typeDefHandle {registeredTypeIndex=1}, varData []}}
      SamplingRegions []
     }}
   )
   &ShapeLocation(
    Cylinder Height=20
    MaskSpawnGroup=true
    Override Seed=false
    Shape Primitive=::NewEnumerator1
    Spawn Group Mask=1
    Surface Only=false
    UseOwnerScale=true
   )
   &ShapeLocation(Hemisphere Distribution=<1 0.5>, MaskSpawnGroup=true, Override Seed=false, Spawn Group Mask=3, Sphere Radius=50, UseOwnerScale=true) {enabled=false, module=ShapeLocation003}
   &ShapeLocation(
    Cylinder Height=1
    Cylinder Height Midpoint=0
    MaskSpawnGroup=true
    Override Seed=false
    Shape Primitive=::NewEnumerator1
    Surface Only=true
    Use Endcaps In Surface Only Mode=false
    UseOwnerScale=true
   ) {enabled=false, module=ShapeLocation001}
   &ShapeLocation(
    Cylinder Height=1
    Cylinder Height Midpoint=0
    MaskSpawnGroup=true
    Override Seed=false
    Shape Primitive=::NewEnumerator1
    Spawn Group Mask=2
    Surface Only=true
    Use Endcaps In Surface Only Mode=false
    UseOwnerScale=true
   ) {enabled=false, module=ShapeLocation002}
   &AddVelocity(
    Velocity=&MaskVectorBySpawnGroup(Masked Vector=$Particles.ShapeLocation.ShapeNormal, Passthrough Vector=$Particles.ShapeLocation003.ShapeNormal, Spawn Group Mask=1)
    Velocity Speed Scale=&MaskFloatBySpawnGroup(Masked Float=0, Passthrough Float=&RandomRangeFloat(Minimum=200, Maximum=500))
   ) {enabled=false}
   &AddVelocity(
    Module.Use Falloff Distance=false
    Origin Offset=<0 0 -50>
    Override Seed=false
    Velocity Mode=::NewEnumerator1
    Velocity Speed=&MaskFloatBySpawnGroup(
     Masked Float=75
     Passthrough Float=&MaskFloatBySpawnGroup(Masked Float=75, Passthrough Float=0, Spawn Group Mask=2)
    )
   ) {enabled=false, module=AddVelocity001}
   &CameraOffset(Camera Offset Amount=75)
   &AddVelocity() {module=AddVelocity002}
  ]
  particleUpdateScript [
   &ParticleState()
   &ScaleSpriteSize(
    Scale Sprite Size Mode=::NewEnumerator0
    Uniform Scale Factor=&MaskFloatBySpawnGroup(
     Masked Float=&FloatFromCurve(FloatCurve=DataInterface() {propertyValues {Curve {keys [{arriveTangent=6.1888, interpMode=RCIM_Cubic, leaveTangent=6.1888, tangentMode=RCTM_User, time=0, value=0}, <at 0.26 1 RCIM_Cubic>]}}})
     Passthrough Float=1
    )
   )
   &ScaleColor(
    Scale Alpha=&FloatFromCurve(
     FloatCurve=DataInterface() {propertyValues {Curve {keys [<at 0 1>,<at 1 0>]}}}
     Scale Curve=&FNTimeOfDayScalar(Night Scale=0.3, Morning Scale=0.6, Day Scale=0.2, Evening Scale=0.3)
    )
    Scale RGB=&MakeVectorFromLinearColorRGB(Color=&FNTimeOfDayColor(
      Night Color=&ColorFromCurve(ColorCurve=DataInterface() {propertyValues {
         AlphaCurve {keys []}
         BlueCurve {keys [<at 0 0.5 RCIM_Cubic>,<at 1 0.3552 RCIM_Cubic>]}
         GreenCurve {keys [<at 0 0.3598 RCIM_Cubic>,<at 1 0.2628 RCIM_Cubic>]}
         RedCurve {keys [<at 0 0.3076 RCIM_Cubic>,<at 1 0.2315 RCIM_Cubic>]}
        }})
      Morning Color=&ColorFromCurve(ColorCurve=DataInterface() {propertyValues {
         AlphaCurve {keys []}
         BlueCurve {keys [<at 0 0.5104 RCIM_Cubic>,<at 1 0.3552 RCIM_Cubic>]}
         GreenCurve {keys [<at 0 0.5104 RCIM_Cubic>,<at 1 0.3552 RCIM_Cubic>]}
         RedCurve {keys [<at 0 0.5104 RCIM_Cubic>,<at 1 0.3552 RCIM_Cubic>]}
        }})
      Day Color=&ColorFromCurve(ColorCurve=DataInterface() {propertyValues {
         AlphaCurve {keys []}
         BlueCurve {keys [<at 0 0.75 RCIM_Cubic>,<at 1 0.5 RCIM_Cubic>]}
         GreenCurve {keys [<at 0 0.75 RCIM_Cubic>,<at 1 0.5 RCIM_Cubic>]}
         RedCurve {keys [<at 0 0.75 RCIM_Cubic>,<at 1 0.5 RCIM_Cubic>]}
        }})
      Evening Color=&ColorFromCurve(ColorCurve=DataInterface() {propertyValues {
         AlphaCurve {keys []}
         BlueCurve {keys [<at 0 0.1781 RCIM_Cubic>,<at 1 0.2411 RCIM_Cubic>]}
         GreenCurve {keys [<at 0 0.1982 RCIM_Cubic>,<at 1 0.2747 RCIM_Cubic>]}
         RedCurve {keys [<at 0 0.2552 RCIM_Cubic>,<at 1 0.3552 RCIM_Cubic>]}
        }})
     ))
    ScaleA=true
    ScaleRGB=true
   )
   &GravityForce(Gravity=&RandomRangeVector(Minimum=<0 0 -150>, Maximum=<0 0 -300>))
   &Drag(
    Drag=&FloatFromCurve(
     FloatCurve=DataInterface() {propertyValues {Curve {keys [{interpMode=RCIM_Cubic, tangentMode=RCTM_Break, time=0, value=1}, {interpMode=RCIM_Cubic, tangentMode=RCTM_Break, time=0.8, value=0}]}}}
     Scale Curve=&MaskFloatBySpawnGroup(Masked Float=1, Passthrough Float=4)
    )
    Use Linear Drag=true
    Use Rotational Drag=false
   )
   &PointAttractionForce(
    Attraction Radius=150
    Attraction Strength=&MaskFloatBySpawnGroup(Masked Float=-10, Passthrough Float=0)
    Kill Within Radius=false
    Module.Use Falloff=true
    UseAttractorOffset=false
   ) {enabled=false}
   &DynamicMaterialParameters(
    Index 0 Param 1=&Add_Float(
     A=&RandomRangeFloat(Minimum=0, Maximum=1)
     B=&FloatFromCurve(
      FloatCurve=DataInterface() {propertyValues {Curve {keys [
          {arriveTangent=2.7533, interpMode=RCIM_Cubic, leaveTangent=1.8958, tangentMode=RCTM_User, time=0, value=1}
          {arriveTangent=0.8292, interpMode=RCIM_Cubic, leaveTangent=0.8292, tangentMode=RCTM_User, time=0.321, value=1.3453}
          {arriveTangent=0.0529, interpMode=RCIM_Cubic, leaveTangent=0.2021, tangentMode=RCTM_User, time=1, value=1.5}
         ]}}}
      CurveIndex=&Add_Float(A=&Multiply_Float(A=$Particles.Initial.MaterialRandom, B=0.5), B=$Particles.NormalizedAge)
      Scale Curve=&Lerp_Float(
       A=0.25
       B=0.7
       Alpha=&NormalizeFloat(Value=$Particles.SizeCurve, Min=$Particles.SizeMin, Max=$Particles.SizeMax)
      )
     )
    )
    Index 0 Param 2=&FloatFromCurve(FloatCurve=DataInterface() {propertyValues {Curve {keys [<at 0 1>,<at 1 0>]}}})
    Index 0 Param 3=&FloatFromCurve(
     FloatCurve=DataInterface() {propertyValues {Curve {keys [
         {interpMode=RCIM_Cubic, leaveTangent=16.5652, leaveTangentWeight=0.4385, tangentMode=RCTM_Break, time=0, value=0}
         {arriveTangentWeight=0.0333, interpMode=RCIM_Cubic, leaveTangentWeight=0.9477, tangentMode=RCTM_Break, time=0.1, value=1}
         {arriveTangent=-1.7159, interpMode=RCIM_Cubic, leaveTangent=-1.7159, tangentMode=RCTM_User, time=0.3, value=0.2}
         {arriveTangent=0.2347, interpMode=RCIM_Cubic, leaveTangent=0.2723, tangentMode=RCTM_Break, time=0.6, value=0}
        ]}}}
     Scale Curve=0
    )
    Index 0 Param 4=&FloatFromCurve(FloatCurve=DataInterface() {propertyValues {Curve {keys [{interpMode=RCIM_Cubic, tangentMode=RCTM_Break, time=0.5, value=1}, {interpMode=RCIM_Cubic, tangentMode=RCTM_Break, time=1, value=0}]}}})
    Index 1 Param 1=&FloatFromCurve(
     FloatCurve=DataInterface() {propertyValues {Curve {keys [
         {interpMode=RCIM_Cubic, leaveTangentWeight=0.0833, tangentMode=RCTM_Break, time=0, value=0}
         {arriveTangentWeight=0.0873, interpMode=RCIM_Cubic, tangentMode=RCTM_Break, time=0.25, value=1}
        ]}}}
     Scale Curve=&Lerp_Float(
      A=0.5
      B=0.75
      Alpha=&NormalizeFloat(Value=$Particles.SizeCurve, Min=$Particles.SizeMin, Max=$Particles.SizeMax)
     )
    )
    Index 1 Param 2=&FloatFromCurve(
     FloatCurve=DataInterface() {propertyValues {Curve {keys [
         {interpMode=RCIM_Cubic, tangentMode=RCTM_Break, time=0, value=0.2}
         {interpMode=RCIM_Cubic, leaveTangent=0.9066, tangentMode=RCTM_Break, time=1, value=2}
        ]}}}
     Scale Curve=&RandomRangeFloat(Minimum=0.1, Maximum=0.2)
    )
    Index 1 Param 3=&FloatFromCurve(FloatCurve=DataInterface() {propertyValues {Curve {keys [
         {interpMode=RCIM_Cubic, leaveTangent=6.3917, tangentMode=RCTM_Break, time=0, value=0}
         {arriveTangent=1.7549, interpMode=RCIM_Cubic, leaveTangent=10.6767, tangentMode=RCTM_User, time=0.25, value=1}
        ]}}})
    Index 1 Param 4=&FloatFromCurve(
     FloatCurve=DataInterface() {propertyValues {Curve {keys [{arriveTangent=-0.136, interpMode=RCIM_Cubic, leaveTangent=-0.136, tangentMode=RCTM_User, time=0, value=1}, <at 1 0.95>]}}}
     Scale Curve=&Lerp_Float(
      A=2
      B=5
      Alpha=&NormalizeFloat(Value=$Particles.SizeCurve, Min=$Particles.SizeMin, Max=$Particles.SizeMax)
     )
    )
    Index 2 Param 1=&FloatFromCurve(
     FloatCurve=DataInterface() {propertyValues {Curve {keys [
         {interpMode=RCIM_Cubic, leaveTangent=4.2485, tangentMode=RCTM_Break, time=0, value=0}
         {arriveTangent=1.0548, interpMode=RCIM_Cubic, leaveTangent=1.0548, tangentMode=RCTM_User, time=0.17, value=0.5715}
         <at 1 1>
        ]}}}
     Scale Curve=3
    )
    Index 2 Param 2=&FloatFromCurve(FloatCurve=DataInterface() {propertyValues {Curve {keys [
         {interpMode=RCIM_Cubic, leaveTangentWeight=0.5024, tangentMode=RCTM_User, time=0, value=1}
         {arriveTangentWeight=0.4857, interpMode=RCIM_Cubic, tangentMode=RCTM_User, time=1, value=1}
        ]}}})
    Index 2 Param 3=&FloatFromCurve(
     FloatCurve=DataInterface() {propertyValues {Curve {keys [
         {interpMode=RCIM_Cubic, leaveTangent=-19.1446, leaveTangentWeight=0.1411, tangentMode=RCTM_Break, time=0, value=1}
         {interpMode=RCIM_Cubic, tangentMode=RCTM_Break, time=0.1, value=0}
         {interpMode=RCIM_Cubic, leaveTangent=11.1233, tangentMode=RCTM_Break, time=1, value=1}
        ]}}}
     Scale Curve=5
    )
    Index 2 Param 4=&RandomRangeFloat(Minimum=1, Maximum=1.75)
    Module.Param0WriteEnabled=true
    Module.Param0WriteEnabled1=true
    Module.Param0WriteEnabled2=true
    Module.Param1WriteEnabled=false
    Module.Param1WriteEnabled1=true
    Module.Param1WriteEnabled2=true
    Module.Param2WriteEnabled=true
    Module.Param2WriteEnabled1=true
    Module.Param2WriteEnabled2=true
    Module.Param3WriteEnabled=true
    Module.Param3WriteEnabled1=true
    Module.Param3WriteEnabled2=true
    Write Parameter Index 1=true
    Write Parameter Index 2=true
   )
   &SolveForcesAndVelocity(Clamp Velocity=false, Limit Acceleration=false)
  ]
  rendererClasses [&NiagaraSpriteRendererProperties]
  renderers [{rendererClass=&NiagaraSpriteRendererProperties}]
 }
 NE_Debris_01001 {
  emitterUpdateScript [
   &EmitterState()
   &SpawnBurst_Instantaneous(Spawn Count=30, Use Spawn Probability=false, UseLoopCountLimit=false)
  ]
  particleSpawnScript [
   &InitializeParticle(
    Color=&MakeCustomLinearColorFromBool(
     Bool=$User.bIsFrontEndPreview
     True Color=&RandomRangeLinearColor(Minimum=<rgba 0.3177 0.1972 0.1605 0.35>, Maximum=<rgba 0.3021 0.2182 0.1731 0.5>)
     False Color=&RandomRangeLinearColor(Minimum=<rgba 0.375 0.2073 0.1562 0.35>, Maximum=<rgba 0.375 0.2328 0.1562 0.5>)
    )
    Lifetime Max=1.5
    Lifetime Min=0.5
    Lifetime Mode=::NewEnumerator1
    Mass Mode=::NewEnumerator1
    Position Offset=<0 0 5>
    Position Offset Coordinate Space=::Local
    Sprite Rotation Mode=::NewEnumerator2
    Sprite Size Mode=::NewEnumerator1
    Uniform Sprite Size Min=1
    UsePositionOffset=true
    Write Material Random=true
   )
   &SkeletalMeshLocation(
    Bone Sampling Mode=::NewEnumerator0
    MaskSpawnGroup=false
    OverrideRandomSeed=false
    Skeletal Mesh=DataInterface() {propertyValues {
      DefaultMesh=null
      ExcludeBoneName=root
      OptionalComponentClass=null
      PreviewMesh=null
      SkinningMode=SkinOnTheFly
      SoftSourceActor=null
      SourceMode=Default
      UvSetIndex=0
      WholeMeshLOD=-1
      bExcludeBone=false
      bReadDeformedGeometry=true
      bRequireCurrentFrameData=true
      ComponentTags []
      FilteredBones [
       JNT_MainRock_01
       JNT_MainRock_02
       JNT_MainRock_03
       JNT_OuterRock_01
       JNT_OuterRock_02
       JNT_OuterRock_03
       JNT_OuterRock_04
       JNT_OuterRock_05
       JNT_OuterRock_06
       JNT_OuterRock_07
       JNT_OuterRock_08
       JNT_OuterRock_09
      ]
      FilteredSockets []
      MeshUserParameter {parameter {name=None, typeDefHandle {registeredTypeIndex=1}, varData []}}
      SamplingRegions []
     }}
   )
   &ShapeLocation(MaskSpawnGroup=false, Override Seed=false, Sphere Radius=50, UseOwnerScale=true) {enabled=false}
   &AddVelocity(
    Module.Use Falloff Distance=false
    Override Seed=false
    Velocity Mode=::NewEnumerator1
    Velocity Origin=$Particles.SkeletalMeshLocation.SampledPosition
    Velocity Speed=&RandomRangeFloat(Minimum=50, Maximum=100)
   )
   &AddVelocity(Velocity=&RandomRangeVector(Minimum=<0 0 200>, Maximum=<1 1 300>)) {module=AddVelocity001}
  ]
  particleUpdateScript [
   &ParticleState()
   &GravityForce(Gravity=&Multiply_Vector(
     A=&RandomRangeVector(Minimum=<0 0 -360>, Maximum=<0 0 -980>)
     B=&VectorFromFloat(Value=&FloatFromCurve(FloatCurve=DataInterface() {propertyValues {Curve {defaultValue=1, keys [{tangentMode=RCTM_User, time=0, value=1}, <at 0.5059 4.6795>, {tangentMode=RCTM_User, time=1, value=7}]}}}))
    ))
   &Drag(Use Linear Drag=true, Use Rotational Drag=true)
   &SolveForcesAndVelocity(Clamp Velocity=false, Limit Acceleration=false)
   &ScaleSpriteSize(Scale Factor=&Vector2DFromFloat(Value=&FloatFromCurve(FloatCurve=DataInterface() {propertyValues {Curve {keys [<at 0.5 1 RCIM_Cubic>, {arriveTangent=-3.7562, interpMode=RCIM_Cubic, leaveTangent=-3.7562, tangentMode=RCTM_User, time=1, value=0}]}}})))
  ]
  rendererClasses [&NiagaraSpriteRendererProperties]
  renderers [{rendererClass=&NiagaraSpriteRendererProperties}]
 }
 NE_Debris_02001 {
  emitterUpdateScript [
   &EmitterState()
   &SpawnBurst_Instantaneous(Spawn Count=15, Use Spawn Probability=false, UseLoopCountLimit=false)
  ]
  particleSpawnScript [
   &InitializeParticle(
    Color Maximum=<rgba 0.4921 0.5944 0.7083 1>
    Color Minimum=<rgba 0.5925 0.8193 0.9844 1>
    Color Mode=::NewEnumerator2
    Lifetime Max=1.5
    Lifetime Min=0.5
    Lifetime Mode=::NewEnumerator1
    Mass Mode=::NewEnumerator1
    Mesh Scale Max=<1 1 1>
    Mesh Scale Min=<0.25 0.25 0.25>
    Mesh Scale Mode=::NewEnumerator2
    Position Offset=<0 0 5>
    Position Offset Coordinate Space=::Local
    UsePositionOffset=true
    Write Material Random=true
   )
   &SkeletalMeshLocation(
    Bone Sampling Mode=::NewEnumerator0
    MaskSpawnGroup=false
    OverrideRandomSeed=false
    Skeletal Mesh=DataInterface() {propertyValues {
      DefaultMesh=null
      ExcludeBoneName=root
      OptionalComponentClass=null
      PreviewMesh=null
      SkinningMode=SkinOnTheFly
      SoftSourceActor=null
      SourceMode=Default
      UvSetIndex=0
      WholeMeshLOD=-1
      bExcludeBone=false
      bReadDeformedGeometry=true
      bRequireCurrentFrameData=true
      ComponentTags []
      FilteredBones [
       JNT_MainRock_01
       JNT_MainRock_02
       JNT_MainRock_03
       JNT_OuterRock_01
       JNT_OuterRock_02
       JNT_OuterRock_03
       JNT_OuterRock_04
       JNT_OuterRock_05
       JNT_OuterRock_06
       JNT_OuterRock_07
       JNT_OuterRock_08
       JNT_OuterRock_09
      ]
      FilteredSockets []
      MeshUserParameter {parameter {name=None, typeDefHandle {registeredTypeIndex=1}, varData []}}
      SamplingRegions []
     }}
   )
   &ShapeLocation(
    MaskSpawnGroup=false
    Override Seed=false
    Sphere Radius=&MergeHelper_FloatSwitch(0=30, 1=15, 2=?, 3=?, Index=&SpawnGroup(), Num=::NewEnumerator0)
    UseOwnerScale=true
   ) {enabled=false}
   &AddVelocity(
    Module.Use Falloff Distance=false
    Override Seed=false
    Velocity Mode=::NewEnumerator1
    Velocity Origin=$Particles.SkeletalMeshLocation.SampledPosition
    Velocity Speed=&RandomRangeFloat(Minimum=50, Maximum=100)
   )
   &AddVelocity(Velocity=&RandomRangeVector(Minimum=<0 0 200>, Maximum=<0 0 400>)) {module=AddVelocity001}
   /SetVariables(Particles.MeshIndex=&RandomRangeInt(Minimum=0, Maximum=2)) {isSetParameters=true}
   &InitialMeshOrientation(Use Rotation Vector=false)
  ]
  particleUpdateScript [
   &ParticleState()
   &UpdateMeshOrientation(
    Rotation Rate=&RandomRangeFloat(Minimum=1, Maximum=2)
    Rotation Vector=&RandomRangeVector(Minimum=<-1 -1 -1>, Maximum=<1 1 1>)
   )
   &GravityForce(Gravity=&Multiply_Vector(
     A=&RandomRangeVector(Minimum=<0 0 -360>, Maximum=<0 0 -980>)
     B=&VectorFromFloat(Value=&FloatFromCurve(FloatCurve=DataInterface() {propertyValues {Curve {defaultValue=1, keys [{tangentMode=RCTM_User, time=0, value=1}, <at 0.5662 3.9496>, {tangentMode=RCTM_User, time=1, value=6}]}}}))
    ))
   &Drag(Use Linear Drag=true, Use Rotational Drag=true)
   &SolveForcesAndVelocity(Clamp Velocity=false, Limit Acceleration=false)
   &ScaleMeshSize(Scale Factor=&VectorFromFloat(Value=&FloatFromCurve(FloatCurve=DataInterface() {propertyValues {Curve {keys [
          {arriveTangent=-0.008, interpMode=RCIM_Cubic, leaveTangent=-0.008, tangentMode=RCTM_User, time=0, value=1}
          {arriveTangent=-3.1748, interpMode=RCIM_Cubic, leaveTangent=-3.1748, tangentMode=RCTM_User, time=0.9, value=0}
         ]}}})))
  ]
  rendererClasses [&NiagaraMeshRendererProperties]
  renderers [{rendererClass=&NiagaraMeshRendererProperties, rendererIndex=0}]
 }
}
summary {
 systemName=NS_Emote_CanineCronutDig_Despawn
 emitters [
  {emitterName=NE_Debris_02001, rendererClasses [&NiagaraMeshRendererProperties]}
  {emitterName=NE_Debris_01001, rendererClasses [&NiagaraSpriteRendererProperties]}
  {emitterName=Dust001, rendererClasses [&NiagaraSpriteRendererProperties]}
 ]
 userVariables [
  User.bIsFrontEnd {type=NiagaraBool}
  User.bIsFrontEndPreview {type=NiagaraBool}
  User.DustSpawnCurve {type=NiagaraFloat, defaultValue=1, description=""}
  User.Particle_BG_Color {type=LinearColor, description="", defaultValue {struct=LinearColor, value=<rgba 1 1 1 1>}}
  User.SkeletalMesh {
   type=NiagaraDataInterfaceSkeletalMesh
   defaultValue {
    struct=NiagaraExt_VariableValue_DataInterface
    value {dataInterface=&self:NiagaraDataInterfaceSkeletalMesh_0}
   }
  }
  User.TranslationSpeed {type=NiagaraPosition, description="", defaultValue {}}
  User.TranslationSpeed001 {type=NiagaraPosition, description="", defaultValue {}}
 ]
}
