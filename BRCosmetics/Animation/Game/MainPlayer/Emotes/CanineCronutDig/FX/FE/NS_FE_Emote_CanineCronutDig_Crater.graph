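# Niagara effect for the CanineCronutDig emote's crater, a short burst of dirt, dust and debris with a briefly visible ground mark.
# A scaled-in mesh and a decal/sprite emitter fade dark and glow layers, while one-shot bursts throw speed-stretched sprites, dust, streaks and tumbling mesh chunks.
# Every emitter shown shifts particles by User.TranslationSpeed, some tapering it over lifetime.
# Front-end switches and time-of-day functions adjust some colors and parameters.
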NiagaraSystem /BRCosmetics/Animation/Game/MainPlayer/Emotes/CanineCronutDig/FX/FE/NS_FE_Emote_CanineCronutDig_Crater
dependencies {
 usedDataInterfaces [
  &NiagaraDataInterfaceCurve
  &NiagaraDataInterfaceSpriteRendererInfo
  &NiagaraDataInterfaceVector2DCurve
  &NiagaraDataInterfaceColorCurve
 ]
 usedDynamicInputs []
 usedModules [&SystemState]
}
emitters {
 DirectionalBurst {
  emitterSpawnScript [
   /SetVariables(Emitter.HoldPositionEmitter:NiagaraPosition=/Niagara/DynamicInputs/Helpers/SimulationPosition()) {isSetParameters=true, module=SetVariables_#1}
   /SetVariables(Emitter.FRONT_END_USE_WPO:NiagaraFloat=&FN_FrontEndFloatModifier(Gameplay=1, Front End=0)) {isSetParameters=true}
   /SetVariables(Emitter.FrontEndColorMod:NiagaraFloat=&FN_FrontEndFloatModifier(Gameplay=0.6, Front End=0.5)) {isSetParameters=true}
   /SetVariables(Emitter.FRONTENDHEIGHTMOD:NiagaraFloat=&FN_FrontEndFloatModifier(Gameplay=-0.04, Front End=-0.01)) {isSetParameters=true}
   /SetVariables(Emitter.FrontEndDistanceHoverOffer:NiagaraFloat=&FN_FrontEndFloatModifier(Gameplay=1.4532, Front End=0)) {isSetParameters=true}
  ]
  emitterUpdateScript [
   &EmitterState(Inactive Response=::NewEnumerator1, Life Cycle Mode=::NewEnumerator1, Loop Behavior=::NewEnumerator1, Loop Duration=0.7, UseLoopDelay=false)
   &SpawnBurst_Instantaneous(Use Spawn Probability=false, UseLoopCountLimit=false)
  ]
  particleSpawnScript [
   &InitializeParticle(
    Color=<rgba 0.0723 0.0168 0.0003 1>
    Lifetime=0.7
    Mass Max=2
    Mass Mode=::NewEnumerator3
    Mesh Scale=<6 6 6>
    Mesh Scale Mode=::NewEnumerator4
    Position Offset=<0 30 0.5>
    Sprite Rotation Angle=90
    Sprite Rotation Mode=::NewEnumerator0
    Sprite Size Mode=::NewEnumerator3
    Uniform Sprite Size=350
    UsePositionOffset=true
    Write Material Random=true
   )
   &InitialMeshOrientation(Mesh Orientation Mode=::NewEnumerator2, Use Rotation Vector=false)
  ]
  particleUpdateScript [
   &ParticleState()
   &ScaleMeshSize(Scale Factor=&VectorFromFloat(Value=&FloatFromCurve(FloatCurve=DataInterface() {propertyValues {Curve {keys [<at 0 0>,<at 0.15 1.04>,<at 0.25 1>,<at 0.8294 1>]}}}))) {module=ScaleMeshSize001}
   &ScaleMeshSize()
   &DynamicMaterialParameters(
    Index 0 Param 1=&FloatFromCurve(FloatCurve=DataInterface() {propertyValues {Curve {keys [<at 0 1>,<at 0.15 0>,<at 0.9 0>,<at 1 1>]}}})
    Index 0 Param 2=&FloatFromCurve(FloatCurve=DataInterface() {propertyValues {Curve {keys [<at 0 1>,<at 0.4793 1>,<at 0.5887 1>,<at 1 1>]}}})
    Module.Index0Float0Write=true
    Module.Index0Float1Write=true
    Module.Index0Float2Write=true
    Module.Index0Float3Write=true
   )
   /SetVariables(
    Particles.DarkGlow:LinearColor=&MultiplyLinearColor(
     Linear Color=<rgba 0.0104 0.0104 0.0104 1.3>
     Scale Factor=&MakeVector4(
      X=1
      Y=1
      Z=1
      W=&FloatFromCurve(FloatCurve=DataInterface() {propertyValues {Curve {keys [<at 0 0>,<at 0.06 0>,<at 0.08 1>,<at 0.8 1>,<at 1 0>]}}})
     )
    )
    Particles.GlowColor=&MultiplyLinearColor(
     Linear Color=<rgba 0.5 0.1206 0 1>
     Scale Factor=&MakeVector4(
      X=1
      Y=1
      Z=1
      W=&FloatFromCurve(
       FloatCurve=DataInterface() {propertyValues {Curve {keys [<at 0 0>,<at 0.06 0>,<at 0.08 1>,<at 0.85 1>,<at 0.98 0>]}}}
       Scale Curve=&FNTimeOfDayScalar(Night Scale=0.8, Morning Scale=1, Evening Scale=1)
      )
     )
    )
   ) {isSetParameters=true, module=SetVariables_#2}
   &ScaleColor(
    Scale Alpha=&FloatFromCurve(FloatCurve=DataInterface() {propertyValues {Curve {keys [<at 0 0>,<at 0.06 0>,<at 0.08 1>,<at 0.9099 1>,<at 1 0>]}}})
    ScaleA=true
    ScaleRGB=true
   )
   &OffsetPosition(Position Offset=<0 0 1>) {enabled=false}
   /SetVariables(Particles.Position=&AddVectorToPosition(
     Position=$Particles.Position
     Vector=&Multiply_Vector(A=$User.TranslationSpeed, B=&VectorFromFloat(Value=$Engine.DeltaTime))
    )) {isSetParameters=true}
  ]
  rendererClasses [&NiagaraMeshRendererProperties]
  renderers [
   {rendererClass=&NiagaraMeshRendererProperties, rendererIndex=0}
   {rendererClass=&NiagaraMeshRendererProperties, rendererIndex=1}
   {rendererClass=&NiagaraMeshRendererProperties, rendererIndex=2}
  ]
 }
 DirectionalBurst001 {
  emitterUpdateScript [
   &EmitterState(Inactive Response=::NewEnumerator1, Life Cycle Mode=::NewEnumerator1, Loop Behavior=::NewEnumerator1, Loop Duration=1, UseLoopDelay=false)
   &SpawnBurst_Instantaneous(Spawn Count=75, Spawn Time=0.05, Use Spawn Probability=false, UseLoopCountLimit=false)
  ]
  particleSpawnScript [
   &InitializeParticle(
    Color=<rgba 10 10 10 1>
    Lifetime Max=0.8
    Lifetime Min=0.2
    Lifetime Mode=::NewEnumerator1
    Mass Max=2
    Mass Mode=::NewEnumerator3
    Position Offset=<0 -30 0>
    Sprite Size Max=<3 14>
    Sprite Size Min=<2 6>
    Sprite Size Mode=::NewEnumerator2
    UsePositionOffset=true
    Write Material Random=true
   )
   &AddVelocity(
    Cone Angle=130
    Cone Axis=&Multiply_Vector(A=$Engine.Owner.SystemYAxis, B=<0.2 0.2 0.2>)
    Distribution Along Cone Axis=0.2
    Module.UseConeFalloff=false
    Module.UseVelDistribution=true
    Override Seed=false
    Rotation Coordinate Space=::Simulation
    Velocity Mode=::NewEnumerator2
    Velocity Speed=&RandomRangeFloat(Minimum=150, Maximum=500)
   )
   &AddVelocity(Velocity=<0 0 300>) {module=AddVelocity001}
   &ShapeLocation(MaskSpawnGroup=false, Override Seed=false, Sphere Radius=30, UseOwnerScale=true)
  ]
  particleUpdateScript [
   &ParticleState()
   &GravityForce()
   &Drag(Drag=&RandomRangeFloat(Minimum=0.8, Maximum=1.2), Use Linear Drag=true, Use Rotational Drag=false)
   &ScaleColor(
    Scale Alpha=&FloatFromCurve(FloatCurve=DataInterface() {propertyValues {Curve {keys [<at -0.0046 1.0113 RCIM_Cubic>,<at 1.0011 0.002 RCIM_Cubic>]}}})
    ScaleA=true
    ScaleRGB=true
   )
   &SolveForcesAndVelocity(Clamp Velocity=false, Limit Acceleration=false)
   &ScaleSpriteSizeBySpeed(
    Max Scale Factor=<1.2 3.3>
    Min Scale Factor=<0.75 0.3>
    Sample Scale Factor By Curve=true
    Scale Factor Curve=DataInterface() {propertyValues {Curve {keys [
        {arriveTangent=2.865, interpMode=RCIM_Cubic, time=0.1247, value=-0.0006}
        {arriveTangent=4.1522, interpMode=RCIM_Cubic, leaveTangent=4.1522, tangentMode=RCTM_User, time=0.3948, value=0.4893}
        <at 0.9958 0.9974 RCIM_Cubic>
       ]}}}
    Velocity Threshold=1500
   )
   &SubUVAnimation(
    Random Start Frame=true
    "Sprite Renderer "=DataInterface() {propertyValues {SpriteRenderer=&self:NiagaraEmitter_1.NiagaraSpriteRendererProperties_0}}
    SubUV Lookup Index Scale=0
    UseEndFrame=false
    UseStartFrame=false
   )
   /SetVariables(Particles.Position=&AddVectorToPosition(
     Position=$Particles.Position
     Vector=&Multiply_Vector(
      A=&Multiply_VectorByFloat(
       Vector=$User.TranslationSpeed
       Float=&FloatFromCurve(FloatCurve=DataInterface() {propertyValues {Curve {keys [<at 0 1>,<at 1 -0.5>]}}})
      )
      B=&VectorFromFloat(Value=$Engine.DeltaTime)
     )
    )) {isSetParameters=true}
  ]
  rendererClasses [&NiagaraSpriteRendererProperties]
  renderers [{rendererClass=&NiagaraSpriteRendererProperties}]
 }
 DirectionalBurst002 {
  emitterUpdateScript [
   &EmitterState(
    Inactive Response=::NewEnumerator1
    Life Cycle Mode=::NewEnumerator1
    Loop Behavior=::NewEnumerator1
    Loop Duration=3.0667
    UseLoopDelay=false
   )
   &SpawnBurst_Instantaneous(Spawn Count=75, Spawn Time=0.05, Use Spawn Probability=false, UseLoopCountLimit=false)
  ]
  particleSpawnScript [
   &InitializeParticle(
    Color=<rgba 10 10 10 1>
    Lifetime Max=0.8
    Lifetime Min=0.2
    Lifetime Mode=::NewEnumerator1
    Mass Max=2
    Mass Mode=::NewEnumerator3
    Position Offset=<0 -30 0>
    Sprite Size Max=<3 14>
    Sprite Size Min=<2 6>
    Sprite Size Mode=::NewEnumerator2
    UsePositionOffset=true
    Write Material Random=true
   )
   &AddVelocity(
    Cone Angle=130
    Cone Axis=<0 0 1>
    Distribution Along Cone Axis=0.2
    Module.UseConeFalloff=false
    Module.UseVelDistribution=true
    Override Seed=false
    Rotation Coordinate Space=::Simulation
    Velocity Mode=::NewEnumerator2
    Velocity Speed=&RandomRangeFloat(Minimum=150, Maximum=500)
   )
   &ShapeLocation(MaskSpawnGroup=false, Override Seed=false, Sphere Radius=30, UseOwnerScale=true)
  ]
  particleUpdateScript [
   &ParticleState()
   &GravityForce()
   &Drag(Drag=&RandomRangeFloat(Minimum=0.8, Maximum=1.2), Use Linear Drag=true, Use Rotational Drag=false)
   &ScaleColor(
    Scale Alpha=&FloatFromCurve(FloatCurve=DataInterface() {propertyValues {Curve {keys [<at -0.0046 1.0113 RCIM_Cubic>,<at 1.0011 0.002 RCIM_Cubic>]}}})
    ScaleA=true
    ScaleRGB=true
   )
   &SolveForcesAndVelocity(Clamp Velocity=false, Limit Acceleration=false)
   &ScaleSpriteSizeBySpeed(
    Max Scale Factor=<1.2 3.3>
    Min Scale Factor=<0.75 0.3>
    Sample Scale Factor By Curve=true
    Scale Factor Curve=DataInterface() {propertyValues {Curve {keys [
        {arriveTangent=2.865, interpMode=RCIM_Cubic, time=0.1247, value=-0.0006}
        {arriveTangent=4.1522, interpMode=RCIM_Cubic, leaveTangent=4.1522, tangentMode=RCTM_User, time=0.3948, value=0.4893}
        <at 0.9958 0.9974 RCIM_Cubic>
       ]}}}
    Velocity Threshold=1500
   )
   &SubUVAnimation(
    Random Start Frame=true
    "Sprite Renderer "=DataInterface() {propertyValues {SpriteRenderer=&self:NiagaraEmitter_2.NiagaraSpriteRendererProperties_0}}
    SubUV Lookup Index Scale=0
    UseEndFrame=false
    UseStartFrame=false
   )
   /SetVariables(Particles.Position=&AddVectorToPosition(
     Position=$Particles.Position
     Vector=&Multiply_Vector(
      A=&Multiply_VectorByFloat(
       Vector=$User.TranslationSpeed
       Float=&FloatFromCurve(FloatCurve=DataInterface() {propertyValues {Curve {keys [<at 0 1>,<at 1 -0.5>]}}})
      )
      B=&VectorFromFloat(Value=$Engine.DeltaTime)
     )
    )) {isSetParameters=true}
  ]
  rendererClasses [&NiagaraSpriteRendererProperties]
  renderers [{rendererClass=&NiagaraSpriteRendererProperties}]
 }
 Dust001 {
  emitterUpdateScript [
   &EmitterState()
   &SpawnBurst_Instantaneous(
    Spawn Count=&RandomRangeInt(Minimum=8, Maximum=12)
    Use Spawn Probability=false
    UseLoopCountLimit=false
   )
   &SpawnBurst_Instantaneous(
    Spawn Count=&RandomRangeInt(Minimum=8, Maximum=12)
    Spawn Group=2
    Use Spawn Probability=false
    UseLoopCountLimit=false
   ) {module=SpawnBurst_Instantaneous002}
   &SpawnBurst_Instantaneous(
    Spawn Count=&RandomRangeInt(Minimum=8, Maximum=12)
    Spawn Group=1
    Spawn Time=0.1
    Use Spawn Probability=false
    UseLoopCountLimit=false
   ) {module=SpawnBurst_Instantaneous001}
   &SpawnBurst_Instantaneous(
    Spawn Count=&RandomRangeInt(Minimum=8, Maximum=12)
    Spawn Group=3
    Spawn Time=0.1
    Use Spawn Probability=false
    UseLoopCountLimit=false
   ) {module=SpawnBurst_Instantaneous003}
  ]
  particleSpawnScript [
   /SetVariables(Particles.VisibilityTag=&MakeCustomIntFromBool(
     Bool=&SetBoolByIntComparison(A=&SpawnGroup(), B=2, Comparison Type=::NewEnumerator2)
     True INT=1
     False INT=0
    )) {isSetParameters=true}
   /SetVariables(
    Particles.SizeMax:NiagaraFloat=&MakeCustomFloatFromBool(
     Bool=&SetBoolByIntComparison(A=&SpawnGroup(), B=2, Comparison Type=::NewEnumerator2)
     True Float=&RandomRangeFloat(Minimum=45, Maximum=65)
     False Float=&RandomRangeFloat(Minimum=155, Maximum=185)
    )
    Particles.SizeMin:NiagaraFloat=&MakeCustomFloatFromBool(
     Bool=&SetBoolByIntComparison(A=&SpawnGroup(), B=2, Comparison Type=::NewEnumerator2)
     True Float=&RandomRangeFloat(Minimum=65, Maximum=85)
     False Float=&RandomRangeFloat(Minimum=165, Maximum=185)
    )
   ) {isSetParameters=true}
   /SetVariables(Particles.SizeCurve:NiagaraFloat=&Lerp_Float(
     A=$Particles.SizeMin
     B=$Particles.SizeMax
     Alpha=&FloatFromCurve(
      FloatCurve=DataInterface() {propertyValues {Curve {keys [
          {interpMode=RCIM_Cubic, tangentMode=RCTM_Break, time=0, value=0}
          {arriveTangentWeight=0.9217, interpMode=RCIM_Cubic, tangentMode=RCTM_Break, time=1, value=1}
         ]}}}
      CurveIndex=&RandomRangeFloat(Minimum=0, Maximum=1)
     )
    )) {isSetParameters=true}
   &InitializeParticle(
    Color=&MakeLinearColorFromVectorAndFloat(
     "Vector (RGB)"=&MakeVectorFromLinearColorRGB(Color=&FNTimeOfDayColor(
       Night Color=<rgba 0.224 0.0705 0.0642 1>
       Morning Color=<rgba 0.1198 0.0244 0.0205 1>
       Day Color=<rgba 0.1198 0.0244 0.0205 1>
       Evening Color=<rgba 0.1198 0.0244 0.0205 1>
      ))
     "Float (Alpha)"=&MaskFloatBySpawnGroup(
      Masked Float=&RandomRangeFloat(Minimum=0.25, Maximum=0.25)
      Passthrough Float=&MaskFloatBySpawnGroup(
       Masked Float=&RandomRangeFloat(Minimum=1, Maximum=1)
       Passthrough Float=&RandomRangeFloat(Minimum=0.3, Maximum=0.3)
       Spawn Group Mask=1
      )
     )
    )
    Lifetime=&MaskFloatBySpawnGroup(
     Masked Float=&RandomRangeFloat(Minimum=1, Maximum=2)
     Passthrough Float=&MaskFloatBySpawnGroup(Masked Float=&RandomRangeFloat(Minimum=0.5, Maximum=1), Passthrough Float=1, Spawn Group Mask=1)
    )
    Position Offset=<0 -30 0>
    Sprite Rotation Mode=::NewEnumerator2
    Sprite Size Mode=::NewEnumerator3
    Uniform Sprite Size=$Particles.SizeCurve
    UsePositionOffset=true
    Write Material Random=true
   )
   &ShapeLocation(Hemisphere Distribution=<1 0.5>, MaskSpawnGroup=true, Override Seed=false, Spawn Group Mask=1, Sphere Radius=50, UseOwnerScale=true)
   &ShapeLocation(Hemisphere Distribution=<1 0.5>, MaskSpawnGroup=true, Override Seed=false, Spawn Group Mask=3, Sphere Radius=50, UseOwnerScale=true) {module=ShapeLocation003}
   &ShapeLocation(
    Cylinder Height=1
    Cylinder Height Midpoint=0
    MaskSpawnGroup=true
    Override Seed=false
    Shape Primitive=::NewEnumerator1
    Surface Only=true
    Use Endcaps In Surface Only Mode=false
    UseOwnerScale=true
   ) {module=ShapeLocation001}
   &ShapeLocation(
    Cylinder Height=1
    Cylinder Height Midpoint=0
    MaskSpawnGroup=true
    Override Seed=false
    Shape Primitive=::NewEnumerator1
    Spawn Group Mask=2
    Surface Only=true
    Use Endcaps In Surface Only Mode=false
    UseOwnerScale=true
   ) {module=ShapeLocation002}
   &AddVelocity(
    Velocity=&MaskVectorBySpawnGroup(Masked Vector=$Particles.ShapeLocation.ShapeNormal, Passthrough Vector=$Particles.ShapeLocation003.ShapeNormal, Spawn Group Mask=1)
    Velocity Speed Scale=&MaskFloatBySpawnGroup(Masked Float=0, Passthrough Float=&RandomRangeFloat(Minimum=200, Maximum=500))
   )
   &AddVelocity(
    Module.Use Falloff Distance=false
    Origin Offset=<0 -30 -50>
    Override Seed=false
    Velocity Mode=::NewEnumerator1
    Velocity Speed=&MaskFloatBySpawnGroup(
     Masked Float=75
     Passthrough Float=&MaskFloatBySpawnGroup(Masked Float=75, Passthrough Float=0, Spawn Group Mask=2)
    )
   ) {module=AddVelocity001}
   &CameraOffset(Camera Offset Amount=75)
  ]
  particleUpdateScript [
   &ParticleState()
   &ScaleSpriteSize(
    Scale Sprite Size Mode=::NewEnumerator0
    Uniform Scale Factor=&MaskFloatBySpawnGroup(
     Masked Float=&FloatFromCurve(FloatCurve=DataInterface() {propertyValues {Curve {keys [{arriveTangent=6.1888, interpMode=RCIM_Cubic, leaveTangent=6.1888, tangentMode=RCTM_User, time=0, value=0}, <at 0.26 1 RCIM_Cubic>]}}})
     Passthrough Float=1
    )
   )
   &ScaleColor(
    Scale Alpha=&FloatFromCurve(
     FloatCurve=DataInterface() {propertyValues {Curve {keys [<at 0 1>,<at 1 0>]}}}
     Scale Curve=&FNTimeOfDayScalar(Night Scale=0.3, Morning Scale=0.6, Day Scale=0.2, Evening Scale=0.3)
    )
    Scale RGB=&MakeVectorFromLinearColorRGB(Color=&FNTimeOfDayColor(
      Night Color=&ColorFromCurve(ColorCurve=DataInterface() {propertyValues {
         AlphaCurve {keys []}
         BlueCurve {keys [<at 0 0.5 RCIM_Cubic>,<at 1 0.3552 RCIM_Cubic>]}
         GreenCurve {keys [<at 0 0.3598 RCIM_Cubic>,<at 1 0.2628 RCIM_Cubic>]}
         RedCurve {keys [<at 0 0.3076 RCIM_Cubic>,<at 1 0.2315 RCIM_Cubic>]}
        }})
      Morning Color=&ColorFromCurve(ColorCurve=DataInterface() {propertyValues {
         AlphaCurve {keys []}
         BlueCurve {keys [<at 0 0.5104 RCIM_Cubic>,<at 1 0.3552 RCIM_Cubic>]}
         GreenCurve {keys [<at 0 0.5104 RCIM_Cubic>,<at 1 0.3552 RCIM_Cubic>]}
         RedCurve {keys [<at 0 0.5104 RCIM_Cubic>,<at 1 0.3552 RCIM_Cubic>]}
        }})
      Day Color=&ColorFromCurve(ColorCurve=DataInterface() {propertyValues {
         AlphaCurve {keys []}
         BlueCurve {keys [<at 0 0.75 RCIM_Cubic>,<at 1 0.5 RCIM_Cubic>]}
         GreenCurve {keys [<at 0 0.75 RCIM_Cubic>,<at 1 0.5 RCIM_Cubic>]}
         RedCurve {keys [<at 0 0.75 RCIM_Cubic>,<at 1 0.5 RCIM_Cubic>]}
        }})
      Evening Color=&ColorFromCurve(ColorCurve=DataInterface() {propertyValues {
         AlphaCurve {keys []}
         BlueCurve {keys [<at 0 0.1781 RCIM_Cubic>,<at 1 0.2411 RCIM_Cubic>]}
         GreenCurve {keys [<at 0 0.1982 RCIM_Cubic>,<at 1 0.2747 RCIM_Cubic>]}
         RedCurve {keys [<at 0 0.2552 RCIM_Cubic>,<at 1 0.3552 RCIM_Cubic>]}
        }})
     ))
    ScaleA=true
    ScaleRGB=true
   )
   &GravityForce(Gravity=&RandomRangeVector(Minimum=<0 0 -150>, Maximum=<0 0 -300>))
   &Drag(
    Drag=&FloatFromCurve(
     FloatCurve=DataInterface() {propertyValues {Curve {keys [{interpMode=RCIM_Cubic, tangentMode=RCTM_Break, time=0, value=1}, {interpMode=RCIM_Cubic, tangentMode=RCTM_Break, time=0.8, value=0}]}}}
     Scale Curve=&MaskFloatBySpawnGroup(Masked Float=1, Passthrough Float=4)
    )
    Use Linear Drag=true
    Use Rotational Drag=false
   )
   &PointAttractionForce(
    Attraction Radius=150
    Attraction Strength=&MaskFloatBySpawnGroup(Masked Float=-10, Passthrough Float=0)
    Kill Within Radius=false
    Module.Use Falloff=true
    UseAttractorOffset=false
   )
   &DynamicMaterialParameters(
    Index 0 Param 1=&Add_Float(
     A=&RandomRangeFloat(Minimum=0, Maximum=1)
     B=&FloatFromCurve(
      FloatCurve=DataInterface() {propertyValues {Curve {keys [
          {arriveTangent=2.7533, interpMode=RCIM_Cubic, leaveTangent=1.8958, tangentMode=RCTM_User, time=0, value=1}
          {arriveTangent=0.8292, interpMode=RCIM_Cubic, leaveTangent=0.8292, tangentMode=RCTM_User, time=0.321, value=1.3453}
          {arriveTangent=0.0529, interpMode=RCIM_Cubic, leaveTangent=0.2021, tangentMode=RCTM_User, time=1, value=1.5}
         ]}}}
      CurveIndex=&Add_Float(A=&Multiply_Float(A=$Particles.Initial.MaterialRandom, B=0.5), B=$Particles.NormalizedAge)
      Scale Curve=&Lerp_Float(
       A=0.25
       B=0.7
       Alpha=&NormalizeFloat(Value=$Particles.SizeCurve, Min=$Particles.SizeMin, Max=$Particles.SizeMax)
      )
     )
    )
    Index 0 Param 2=&FloatFromCurve(FloatCurve=DataInterface() {propertyValues {Curve {keys [<at 0 1>,<at 1 0>]}}})
    Index 0 Param 3=&FloatFromCurve(
     FloatCurve=DataInterface() {propertyValues {Curve {keys [
         {interpMode=RCIM_Cubic, leaveTangent=16.5652, leaveTangentWeight=0.4385, tangentMode=RCTM_Break, time=0, value=0}
         {arriveTangentWeight=0.0333, interpMode=RCIM_Cubic, leaveTangentWeight=0.9477, tangentMode=RCTM_Break, time=0.1, value=1}
         {arriveTangent=-1.7159, interpMode=RCIM_Cubic, leaveTangent=-1.7159, tangentMode=RCTM_User, time=0.3, value=0.2}
         {arriveTangent=0.2347, interpMode=RCIM_Cubic, leaveTangent=0.2723, tangentMode=RCTM_Break, time=0.6, value=0}
        ]}}}
     Scale Curve=0
    )
    Index 0 Param 4=&FloatFromCurve(FloatCurve=DataInterface() {propertyValues {Curve {keys [{interpMode=RCIM_Cubic, tangentMode=RCTM_Break, time=0.5, value=1}, {interpMode=RCIM_Cubic, tangentMode=RCTM_Break, time=1, value=0}]}}})
    Index 1 Param 1=&FloatFromCurve(
     FloatCurve=DataInterface() {propertyValues {Curve {keys [
         {interpMode=RCIM_Cubic, leaveTangentWeight=0.0833, tangentMode=RCTM_Break, time=0, value=0}
         {arriveTangentWeight=0.0873, interpMode=RCIM_Cubic, tangentMode=RCTM_Break, time=0.25, value=1}
        ]}}}
     Scale Curve=&Lerp_Float(
      A=0.5
      B=0.75
      Alpha=&NormalizeFloat(Value=$Particles.SizeCurve, Min=$Particles.SizeMin, Max=$Particles.SizeMax)
     )
    )
    Index 1 Param 2=&FloatFromCurve(
     FloatCurve=DataInterface() {propertyValues {Curve {keys [
         {interpMode=RCIM_Cubic, tangentMode=RCTM_Break, time=0, value=0.2}
         {interpMode=RCIM_Cubic, leaveTangent=0.9066, tangentMode=RCTM_Break, time=1, value=2}
        ]}}}
     Scale Curve=&RandomRangeFloat(Minimum=0.1, Maximum=0.2)
    )
    Index 1 Param 3=&FloatFromCurve(FloatCurve=DataInterface() {propertyValues {Curve {keys [
         {interpMode=RCIM_Cubic, leaveTangent=6.3917, tangentMode=RCTM_Break, time=0, value=0}
         {arriveTangent=1.7549, interpMode=RCIM_Cubic, leaveTangent=10.6767, tangentMode=RCTM_User, time=0.25, value=1}
        ]}}})
    Index 1 Param 4=&FloatFromCurve(
     FloatCurve=DataInterface() {propertyValues {Curve {keys [{arriveTangent=-0.136, interpMode=RCIM_Cubic, leaveTangent=-0.136, tangentMode=RCTM_User, time=0, value=1}, <at 1 0.95>]}}}
     Scale Curve=&Lerp_Float(
      A=2
      B=5
      Alpha=&NormalizeFloat(Value=$Particles.SizeCurve, Min=$Particles.SizeMin, Max=$Particles.SizeMax)
     )
    )
    Index 2 Param 1=&FloatFromCurve(
     FloatCurve=DataInterface() {propertyValues {Curve {keys [
         {interpMode=RCIM_Cubic, leaveTangent=4.2485, tangentMode=RCTM_Break, time=0, value=0}
         {arriveTangent=1.0548, interpMode=RCIM_Cubic, leaveTangent=1.0548, tangentMode=RCTM_User, time=0.17, value=0.5715}
         <at 1 1>
        ]}}}
     Scale Curve=3
    )
    Index 2 Param 2=&FloatFromCurve(FloatCurve=DataInterface() {propertyValues {Curve {keys [
         {interpMode=RCIM_Cubic, leaveTangentWeight=0.5024, tangentMode=RCTM_User, time=0, value=1}
         {arriveTangentWeight=0.4857, interpMode=RCIM_Cubic, tangentMode=RCTM_User, time=1, value=1}
        ]}}})
    Index 2 Param 3=&FloatFromCurve(
     FloatCurve=DataInterface() {propertyValues {Curve {keys [
         {interpMode=RCIM_Cubic, leaveTangent=-19.1446, leaveTangentWeight=0.1411, tangentMode=RCTM_Break, time=0, value=1}
         {interpMode=RCIM_Cubic, tangentMode=RCTM_Break, time=0.1, value=0}
         {interpMode=RCIM_Cubic, leaveTangent=11.1233, tangentMode=RCTM_Break, time=1, value=1}
        ]}}}
     Scale Curve=5
    )
    Index 2 Param 4=&RandomRangeFloat(Minimum=1, Maximum=1.75)
    Module.Param0WriteEnabled=true
    Module.Param0WriteEnabled1=true
    Module.Param0WriteEnabled2=true
    Module.Param1WriteEnabled=false
    Module.Param1WriteEnabled1=true
    Module.Param1WriteEnabled2=true
    Module.Param2WriteEnabled=true
    Module.Param2WriteEnabled1=true
    Module.Param2WriteEnabled2=true
    Module.Param3WriteEnabled=true
    Module.Param3WriteEnabled1=true
    Module.Param3WriteEnabled2=true
    Write Parameter Index 1=true
    Write Parameter Index 2=true
   )
   &SolveForcesAndVelocity(Clamp Velocity=false, Limit Acceleration=false)
   /SetVariables(Particles.Position=&AddVectorToPosition(
     Position=$Particles.Position
     Vector=&Multiply_Vector(A=$User.TranslationSpeed, B=&VectorFromFloat(Value=$Engine.DeltaTime))
    )) {isSetParameters=true}
  ]
  rendererClasses [&NiagaraSpriteRendererProperties]
  renderers [
   {rendererClass=&NiagaraSpriteRendererProperties}
   {rendererClass=&NiagaraSpriteRendererProperties, rendererIndex=1}
  ]
 }
 Low_Med {
  emitterSpawnScript [/SetVariables(Emitter.HoldPositionEmitter:NiagaraPosition=&SimulationPosition()) {isSetParameters=true, module=SetVariables_#1}]
  emitterUpdateScript [
   &EmitterState(Inactive Response=::NewEnumerator1, Life Cycle Mode=::NewEnumerator1, Loop Behavior=::NewEnumerator1, Loop Duration=0.7, UseLoopDelay=false)
   &SpawnBurst_Instantaneous(Use Spawn Probability=false, UseLoopCountLimit=false)
   /SetVariables(Emitter.OpacityMult:NiagaraFloat=&FloatFromCurve(
     FloatCurve=DataInterface() {propertyValues {Curve {keys [<at 0 1.2>,<at 0.1 0>,<at 0.9 0>,<at 1 1.2>]}}}
     CurveIndex=$Emitter.NormalizedLoopAge
    )) {isSetParameters=true}
  ]
  particleSpawnScript [
   &InitializeParticle(
    Color=<rgba 0.0103 0.0024 0 1>
    Lifetime=0.7
    Mass Max=2
    Mass Mode=::NewEnumerator3
    Mesh Scale=<6 6 6>
    Mesh Scale Mode=::NewEnumerator4
    Position Offset=<0 30 1>
    Sprite Rotation Angle=119
    Sprite Rotation Mode=::NewEnumerator0
    Sprite Size Mode=::NewEnumerator3
    Uniform Sprite Size=360
    UsePositionOffset=true
    Write Material Random=true
   )
   /Niagara/Modules/Decal/Decal_Attributes(
    Decal Alignment:Vector3f=$Engine.Owner.SystemYAxis
    Decal Facing:Vector3f=$Engine.Owner.SystemZAxis
    Decal Facing Mode=1
    Decal Rotation Mode=3
    Decal Size:Vector3f=<500 360 360>
    Orientation=1
    Rotation Angle=0.33
    Set Alignment=true
    Size=1
   )
   &SpriteFacingAndAlignment(Sprite Facing=<0 0 1>, WriteAlignment=true, WriteFacing=true)
   /SetVariables(Particles.VisibilityTag=&MakeIntFromFloat(Float=&FN_FrontEndFloatModifier(Gameplay=1, Front End=1))) {isSetParameters=true}
  ]
  particleUpdateScript [
   &ParticleState()
   &DynamicMaterialParameters(
    Index 0 Param 1=&FloatFromCurve(FloatCurve=DataInterface() {propertyValues {Curve {keys [<at 0 1>,<at 0.05 0>,<at 0.9 0>,<at 1 1>]}}})
    Index 0 Param 2=&FloatFromCurve(FloatCurve=DataInterface() {propertyValues {Curve {keys [<at 0 1>,<at 0.4793 1>,<at 0.5887 1>,<at 1 1>]}}})
    Module.Index0Float0Write=true
    Module.Index0Float1Write=true
    Module.Index0Float2Write=true
    Module.Index0Float3Write=true
   )
   /SetVariables(
    Particles.DarkGlow:LinearColor=&MultiplyLinearColor(
     Linear Color=<rgba 0.0104 0.0104 0.0104 10>
     Scale Factor=&MakeVector4(
      X=1
      Y=1
      Z=1
      W=&FloatFromCurve(FloatCurve=DataInterface() {propertyValues {Curve {keys [<at 0 1>,<at 0.8 1>,<at 1 0>]}}}, Scale Curve=5)
     )
    )
    Particles.GlowColor=&MultiplyLinearColor(
     Linear Color=<rgba 0.5 0.1206 0 1>
     Scale Factor=&MakeVector4(
      X=1
      Y=1
      Z=1
      W=&FloatFromCurve(
       FloatCurve=DataInterface() {propertyValues {Curve {keys [<at 0 0>,<at 0.2 1>,<at 0.85 1>,<at 0.98 0>]}}}
       Scale Curve=&FNTimeOfDayScalar(Night Scale=0.8, Morning Scale=1, Evening Scale=1)
      )
     )
    )
   ) {isSetParameters=true, module=SetVariables_#2}
   &ScaleColor(
    Scale Alpha=&FloatFromCurve(FloatCurve=DataInterface() {propertyValues {Curve {keys [<at 0 0>,<at 0.06 0>,<at 0.08 1>,<at 0.9099 1>,<at 1 0>]}}})
    ScaleA=true
    ScaleRGB=true
   )
   /SetVariables(Particles.Position=&AddVectorToPosition(
     Position=$Particles.Position
     Vector=&Multiply_Vector(A=$User.TranslationSpeed, B=&VectorFromFloat(Value=$Engine.DeltaTime))
    )) {isSetParameters=true}
  ]
  rendererClasses [&NiagaraDecalRendererProperties, &NiagaraSpriteRendererProperties]
  renderers [
   {rendererClass=&NiagaraDecalRendererProperties, rendererIndex=0}
   {rendererClass=&NiagaraDecalRendererProperties, rendererIndex=1}
   {rendererClass=&NiagaraDecalRendererProperties, rendererIndex=2}
   {rendererClass=&NiagaraSpriteRendererProperties, rendererIndex=3}
   {rendererClass=&NiagaraSpriteRendererProperties, rendererIndex=4}
   {rendererClass=&NiagaraSpriteRendererProperties, rendererIndex=5}
  ]
 }
 NE_Debris_01001 {
  emitterUpdateScript [
   &EmitterState()
   &SpawnBurst_Instantaneous(Spawn Count=50, Spawn Time=0.05, Use Spawn Probability=false, UseLoopCountLimit=false)
  ]
  particleSpawnScript [
   &InitializeParticle(
    Color=&MakeCustomLinearColorFromBool(
     Bool=$User.bIsFrontEndPreview
     True Color=&RandomRangeLinearColor(Minimum=<rgba 0.3177 0.1972 0.1605 0.35>, Maximum=<rgba 0.3021 0.2182 0.1731 0.5>)
     False Color=&RandomRangeLinearColor(Minimum=<rgba 0.375 0.2073 0.1562 0.35>, Maximum=<rgba 0.375 0.2328 0.1562 0.5>)
    )
    Lifetime Max=1.5
    Lifetime Min=0.5
    Lifetime Mode=::NewEnumerator1
    Mass Mode=::NewEnumerator1
    Position Offset=<0 -30 5>
    Position Offset Coordinate Space=::Local
    Sprite Rotation Mode=::NewEnumerator2
    Sprite Size Mode=::NewEnumerator1
    Uniform Sprite Size Min=1
    UsePositionOffset=true
    Write Material Random=true
   )
   &ShapeLocation(MaskSpawnGroup=false, Override Seed=false, Sphere Radius=50, UseOwnerScale=true)
   &AddVelocity(
    Module.Use Falloff Distance=false
    Origin Offset=<0 -30 0>
    Override Seed=false
    Velocity Mode=::NewEnumerator1
    Velocity Speed=&RandomRangeFloat(Minimum=150, Maximum=300)
   )
   &AddVelocity(Velocity=&RandomRangeVector(Minimum=<0 0 200>, Maximum=<1 1 300>)) {module=AddVelocity001}
  ]
  particleUpdateScript [
   &ParticleState()
   &GravityForce(Gravity=&RandomRangeVector(Minimum=<0 0 -360>, Maximum=<1 1 -980>))
   &Drag(Use Linear Drag=true, Use Rotational Drag=true)
   &SolveForcesAndVelocity(Clamp Velocity=false, Limit Acceleration=false)
   &ScaleSpriteSize(Scale Factor=&Vector2DFromFloat(Value=&FloatFromCurve(FloatCurve=DataInterface() {propertyValues {Curve {keys [<at 0.5 1 RCIM_Cubic>, {arriveTangent=-3.7562, interpMode=RCIM_Cubic, leaveTangent=-3.7562, tangentMode=RCTM_User, time=1, value=0}]}}})))
   /SetVariables(Particles.Position=&AddVectorToPosition(
     Position=$Particles.Position
     Vector=&Multiply_Vector(
      A=&Multiply_VectorByFloat(Vector=$User.TranslationSpeed, Float=1)
      B=&VectorFromFloat(Value=$Engine.DeltaTime)
     )
    )) {isSetParameters=true}
  ]
  rendererClasses [&NiagaraSpriteRendererProperties]
  renderers [{rendererClass=&NiagaraSpriteRendererProperties}]
 }
 NE_Debris_02 {
  emitterUpdateScript [
   &EmitterState()
   &SpawnBurst_Instantaneous(Spawn Count=3, Spawn Time=0.05, Use Spawn Probability=false, UseLoopCountLimit=false)
  ]
  particleSpawnScript [
   &InitializeParticle(
    Color Maximum=<rgba 0.4921 0.5944 0.7083 1>
    Color Minimum=<rgba 0.5925 0.8193 0.9844 1>
    Color Mode=::NewEnumerator2
    Lifetime Max=1.5
    Lifetime Min=0.5
    Lifetime Mode=::NewEnumerator1
    Mass Mode=::NewEnumerator1
    Mesh Scale Max=<3 3 3>
    Mesh Scale Min=<2 2 2>
    Mesh Scale Mode=::NewEnumerator2
    Position Offset=<0 -30 5>
    Position Offset Coordinate Space=::Local
    UsePositionOffset=true
    Write Material Random=true
   )
   &ShapeLocation(
    MaskSpawnGroup=false
    Override Seed=false
    Sphere Radius=&MergeHelper_FloatSwitch(0=30, 1=15, 2=?, 3=?, Index=&SpawnGroup(), Num=::NewEnumerator0)
    UseOwnerScale=true
   )
   &AddVelocity(
    Module.Use Falloff Distance=false
    Origin Offset=<0 -30 0>
    Override Seed=false
    Velocity Mode=::NewEnumerator1
    Velocity Speed=&RandomRangeFloat(Minimum=150, Maximum=300)
   )
   &AddVelocity(Velocity=&RandomRangeVector(Minimum=<0 0 200>, Maximum=<0 0 400>)) {module=AddVelocity001}
   /SetVariables(Particles.MeshIndex=&RandomRangeInt(Minimum=0, Maximum=2)) {isSetParameters=true, module=SetVariables_#3}
   &InitialMeshOrientation(Use Rotation Vector=false)
  ]
  particleUpdateScript [
   &ParticleState()
   &UpdateMeshOrientation(
    Rotation Rate=&RandomRangeFloat(Minimum=1, Maximum=2)
    Rotation Vector=&RandomRangeVector(Minimum=<-1 -1 -1>, Maximum=<1 1 1>)
   )
   &GravityForce(Gravity=&RandomRangeVector(Minimum=<0 0 -360>, Maximum=<1 1 -980>))
   &Drag(Use Linear Drag=true, Use Rotational Drag=true)
   &SolveForcesAndVelocity(Clamp Velocity=false, Limit Acceleration=false)
   &ScaleMeshSize(Scale Factor=&VectorFromFloat(Value=&FloatFromCurve(FloatCurve=DataInterface() {propertyValues {Curve {keys [
          {arriveTangent=-0.008, interpMode=RCIM_Cubic, leaveTangent=-0.008, tangentMode=RCTM_User, time=0, value=1}
          {arriveTangent=-3.1748, interpMode=RCIM_Cubic, leaveTangent=-3.1748, tangentMode=RCTM_User, time=0.9, value=0}
         ]}}})))
   /SetVariables(Particles.Position=&AddVectorToPosition(
     Position=$Particles.Position
     Vector=&Multiply_Vector(
      A=&Multiply_VectorByFloat(
       Vector=$User.TranslationSpeed
       Float=&FloatFromCurve(FloatCurve=DataInterface() {propertyValues {Curve {keys [<at 0 1>,<at 1 0>]}}})
      )
      B=&VectorFromFloat(Value=$Engine.DeltaTime)
     )
    )) {isSetParameters=true}
  ]
  rendererClasses [&NiagaraMeshRendererProperties]
  renderers [{rendererClass=&NiagaraMeshRendererProperties, rendererIndex=0}]
 }
 NE_Debris_02001 {
  emitterUpdateScript [
   &EmitterState()
   &SpawnBurst_Instantaneous(Spawn Count=15, Spawn Time=0.05, Use Spawn Probability=false, UseLoopCountLimit=false)
   &SpawnBurst_Instantaneous(Spawn Count=15, Spawn Group=1, Spawn Time=0.05, Use Spawn Probability=false, UseLoopCountLimit=false) {module=SpawnBurst_Instantaneous001}
  ]
  particleSpawnScript [
   &InitializeParticle(
    Color Maximum=<rgba 0.4921 0.5944 0.7083 1>
    Color Minimum=<rgba 0.5925 0.8193 0.9844 1>
    Color Mode=::NewEnumerator2
    Lifetime Max=1.5
    Lifetime Min=0.5
    Lifetime Mode=::NewEnumerator1
    Mass Mode=::NewEnumerator1
    Mesh Scale Max=<1 1 1>
    Mesh Scale Min=<0.25 0.25 0.25>
    Mesh Scale Mode=::NewEnumerator2
    Position Offset=<0 -30 5>
    Position Offset Coordinate Space=::Local
    UsePositionOffset=true
    Write Material Random=true
   )
   &ShapeLocation(
    MaskSpawnGroup=false
    Override Seed=false
    Sphere Radius=&MergeHelper_FloatSwitch(0=30, 1=15, 2=?, 3=?, Index=&SpawnGroup(), Num=::NewEnumerator0)
    UseOwnerScale=true
   )
   &AddVelocity(
    Module.Use Falloff Distance=false
    Origin Offset=<0 -30 0>
    Override Seed=false
    Velocity Mode=::NewEnumerator1
    Velocity Speed=&RandomRangeFloat(Minimum=150, Maximum=300)
   )
   &AddVelocity(Velocity=&RandomRangeVector(Minimum=<0 0 200>, Maximum=<0 0 400>)) {module=AddVelocity001}
   /SetVariables(Particles.MeshIndex=&RandomRangeInt(Minimum=0, Maximum=2)) {isSetParameters=true, module=SetVariables_#3}
   &InitialMeshOrientation(Use Rotation Vector=false)
  ]
  particleUpdateScript [
   &ParticleState()
   &UpdateMeshOrientation(
    Rotation Rate=&RandomRangeFloat(Minimum=1, Maximum=2)
    Rotation Vector=&RandomRangeVector(Minimum=<-1 -1 -1>, Maximum=<1 1 1>)
   )
   &GravityForce(Gravity=&RandomRangeVector(Minimum=<0 0 -360>, Maximum=<1 1 -980>))
   &Drag(Use Linear Drag=true, Use Rotational Drag=true)
   &SolveForcesAndVelocity(Clamp Velocity=false, Limit Acceleration=false)
   &ScaleMeshSize(Scale Factor=&VectorFromFloat(Value=&FloatFromCurve(FloatCurve=DataInterface() {propertyValues {Curve {keys [
          {arriveTangent=-0.008, interpMode=RCIM_Cubic, leaveTangent=-0.008, tangentMode=RCTM_User, time=0, value=1}
          {arriveTangent=-3.1748, interpMode=RCIM_Cubic, leaveTangent=-3.1748, tangentMode=RCTM_User, time=0.9, value=0}
         ]}}})))
   /SetVariables(Particles.Position=&AddVectorToPosition(
     Position=$Particles.Position
     Vector=&Multiply_Vector(
      A=&Multiply_VectorByFloat(
       Vector=$User.TranslationSpeed
       Float=&FloatFromCurve(FloatCurve=DataInterface() {propertyValues {Curve {keys [<at 0 1>,<at 1 0>]}}})
      )
      B=&VectorFromFloat(Value=$Engine.DeltaTime)
     )
    )) {isSetParameters=true}
  ]
  rendererClasses [&NiagaraMeshRendererProperties]
  renderers [{rendererClass=&NiagaraMeshRendererProperties, rendererIndex=0}]
 }
 NE_Streaks {
  emitterUpdateScript [
   &EmitterState()
   &SpawnBurst_Instantaneous(
    Spawn Count=&RandomRangeInt(Minimum=8, Maximum=12)
    Spawn Time=0.105
    Use Spawn Probability=false
    UseLoopCountLimit=false
   )
  ]
  particleSpawnScript [
   &InitializeParticle(
    Color Maximum=<rgba 0.1198 0.057 0.044 0.502>
    Color Minimum=<rgba 0.1146 0.0244 0.0175 0.251>
    Color Mode=::NewEnumerator2
    Lifetime Max=1
    Lifetime Min=0.5
    Lifetime Mode=::NewEnumerator1
    Position Offset=<0 -30 -30>
    Sprite Size Max=<75 250>
    Sprite Size Min=<35 100>
    Sprite Size Mode=::NewEnumerator2
    UsePositionOffset=true
    Write Material Random=true
   )
   &ShapeLocation(
    Cylinder Height=1
    Cylinder Radius=50
    MaskSpawnGroup=false
    Override Seed=false
    Shape Primitive=::NewEnumerator1
    Surface Only=false
    UseOwnerScale=true
   )
   &AddVelocity(
    Cone Angle=120
    Cone Axis=<0 0 1>
    Module.UseConeFalloff=false
    Module.UseVelDistribution=false
    Override Seed=false
    Velocity Mode=::NewEnumerator2
    Velocity Speed=500
   )
  ]
  particleUpdateScript [
   &ParticleState()
   &CameraOffset(Camera Offset Amount=25)
   &Drag(Drag=10, Use Linear Drag=true, Use Rotational Drag=true)
   &SolveForcesAndVelocity(Clamp Velocity=false, Limit Acceleration=false)
   &DynamicMaterialParameters(
    Index 0 Param 1=100
    Index 0 Param 4=&FloatFromCurve(FloatCurve=DataInterface() {propertyValues {Curve {
        defaultValue=0
        keys [{interpMode=RCIM_Cubic, tangentMode=RCTM_User, time=0, value=0}, {interpMode=RCIM_Cubic, tangentMode=RCTM_User, time=0.4, value=1}]
       }}})
    Index 1 Param 1=&FloatFromCurve(FloatCurve=DataInterface() {propertyValues {Curve {
        defaultValue=1
        keys [
         {interpMode=RCIM_Cubic, tangentMode=RCTM_User, time=0, value=0.01}
         {arriveTangent=1.0027, interpMode=RCIM_Cubic, leaveTangent=1.0027, tangentMode=RCTM_User, time=1, value=0.5}
        ]
       }}})
    Index 1 Param 3=&FloatFromCurve(FloatCurve=DataInterface() {propertyValues {Curve {
        defaultValue=1
        keys [{interpMode=RCIM_Cubic, tangentMode=RCTM_User, time=0.2, value=0}, {interpMode=RCIM_Cubic, tangentMode=RCTM_User, time=1, value=2}]
       }}})
    Index 1 Param 4=&FloatFromCurve(FloatCurve=DataInterface() {propertyValues {Curve {defaultValue=1, keys [{tangentMode=RCTM_User, time=0, value=-0.3}, {tangentMode=RCTM_User, time=1, value=0}]}}})
    Module.Index0Float0Write=true
    Module.Index0Float1Write=true
    Module.Index0Float2Write=true
    Module.Index0Float3Write=true
    Module.Index1Float0Write=true
    Module.Index1Float1Write=true
    Module.Index1Float2Write=true
    Module.Index1Float3Write=true
    Write Parameter Index 1=::NewEnumerator1
   )
   &ScaleColor(
    Scale Alpha=&FloatFromCurve(FloatCurve=DataInterface() {propertyValues {Curve {keys [<at 0 1>,<at 0.5 1>,<at 1 0>]}}})
    Scale RGB=<0.6 0.6 0.6>
    ScaleA=true
    ScaleRGB=true
   )
   &ScaleSpriteSize(
    Non-Uniform Curve Sprite Scale=DataInterface() {propertyValues {
      XCurve {keys [
        {interpMode=RCIM_Cubic, leaveTangent=0.569, tangentMode=RCTM_Break, time=0, value=0.6}
        {arriveTangent=2.7528, interpMode=RCIM_Cubic, tangentMode=RCTM_Break, time=0.9, value=1.75}
       ]}
      YCurve {keys [
        {interpMode=RCIM_Cubic, leaveTangent=8.9163, tangentMode=RCTM_Break, time=0, value=0}
        {interpMode=RCIM_Cubic, leaveTangent=0.5329, tangentMode=RCTM_Break, time=0.2, value=1.5}
        {interpMode=RCIM_Cubic, tangentMode=RCTM_Break, time=1, value=1.79}
       ]}
     }}
    Scale Sprite Size Mode=::NewEnumerator3
   )
   /SetVariables(Particles.Position=&AddVectorToPosition(
     Position=$Particles.Position
     Vector=&Multiply_Vector(
      A=&Multiply_VectorByFloat(Vector=$User.TranslationSpeed, Float=1)
      B=&VectorFromFloat(Value=$Engine.DeltaTime)
     )
    )) {isSetParameters=true}
  ]
  rendererClasses [&NiagaraSpriteRendererProperties]
  renderers [{rendererClass=&NiagaraSpriteRendererProperties}]
 }
}
summary {
 systemName=NS_FE_Emote_CanineCronutDig_Crater
 emitters [
  {emitterName=DirectionalBurst, rendererClasses [&NiagaraMeshRendererProperties]}
  {emitterName=DirectionalBurst001, rendererClasses [&NiagaraSpriteRendererProperties]}
  {emitterName=DirectionalBurst002, rendererClasses [&NiagaraSpriteRendererProperties]}
  {emitterName=NE_Debris_02001, rendererClasses [&NiagaraMeshRendererProperties]}
  {emitterName=NE_Streaks, rendererClasses [&NiagaraSpriteRendererProperties]}
  {emitterName=NE_Debris_01001, rendererClasses [&NiagaraSpriteRendererProperties]}
  {emitterName=Dust001, rendererClasses [&NiagaraSpriteRendererProperties]}
  {emitterName=NE_Debris_02, rendererClasses [&NiagaraMeshRendererProperties]}
  {emitterName=Low_Med, rendererClasses [&NiagaraDecalRendererProperties, &NiagaraSpriteRendererProperties]}
 ]
 userVariables [
  User.bIsFrontEnd {type=NiagaraBool}
  User.bIsFrontEndPreview {type=NiagaraBool}
  User.DustSpawnCurve {type=NiagaraFloat, defaultValue=1, description=""}
  User.Particle_BG_Color {type=LinearColor, description="", defaultValue {struct=LinearColor, value=<rgba 1 1 1 1>}}
  User.SkeletalMesh {
   type=NiagaraDataInterfaceSkeletalMesh
   defaultValue {
    struct=NiagaraExt_VariableValue_DataInterface
    value {dataInterface=&self:NiagaraDataInterfaceSkeletalMesh_0}
   }
  }
  User.TranslationSpeed {type=Vector3f, description="", defaultValue {struct=Vector3f, value=<0 -290 0>}}
 ]
}
